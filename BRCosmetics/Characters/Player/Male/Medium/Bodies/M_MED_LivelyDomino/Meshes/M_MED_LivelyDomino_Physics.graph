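# Physics asset for the M_MED_LivelyDomino male medium player body cosmetic, set up for two dangling scarf chains.
# Each five-capsule chain hangs from a base body angularly locked to spine_05, with tight twist and one swing axis loosening toward the tip.
# Capsules on the pelvis, spine, upper arms and head give the torso shape, linked by uniform-limit constraints except the head, which has none.
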PhysicsAsset /BRCosmetics/Characters/Player/Male/Medium/Bodies/M_MED_LivelyDomino/Meshes/M_MED_LivelyDomino_Physics
bodies {
 dyn_scarf_l_01 [{shapeType=Capsule, length=13, radius=3, shapeName=dyn_scarf_l_01_capsule, center=<5.6113 0.0825 0.0266>, rotation=<rot 0 -88.4583 -86.5263>}]
 dyn_scarf_l_02 [{shapeType=Capsule, length=13, radius=3, shapeName=dyn_scarf_l_02_capsule, center=<9.9615 0.0273 -0.0094>, rotation=<rot 0 89.3657 -85.1148>}]
 dyn_scarf_l_03 [{shapeType=Capsule, length=13, radius=3, shapeName=dyn_scarf_l_03_capsule, center=<10.819 -0.0576 -0.5064>, rotation=<rot 0 89.4527 -86.5692>}]
 dyn_scarf_l_04 [{shapeType=Capsule, length=13, radius=3, shapeName=dyn_scarf_l_04_capsule, center=<9.4456 0.0942 0.7164>, rotation=<rot 0 91.4227 -89.7449>}]
 dyn_scarf_l_05 [{shapeType=Capsule, length=13, radius=3, shapeName=dyn_scarf_l_05_capsule, center=<9.6524 0.6136 -1.4759>, rotation=<rot 0 -88.1072 -73.3068>}]
 dyn_scarf_l_base [{shapeType=Capsule, length=0, radius=5, shapeName=dyn_scarf_l_base_capsule, rotation=<rot 0 -90 0>}]
 dyn_scarf_r_01 [{shapeType=Capsule, length=13, radius=3, shapeName=dyn_scarf_r_01_capsule, center=<4.9921 0.0717 1.0404>, rotation=<rot 0 90.991 -87.3219>}]
 dyn_scarf_r_02 [{shapeType=Capsule, length=13, radius=3, shapeName=dyn_scarf_r_02_capsule, center=<9.4756 0.1802 0.9989>, rotation=<rot 0 91.5526 -87.8417>}]
 dyn_scarf_r_03 [{shapeType=Capsule, length=13, radius=3, shapeName=dyn_scarf_r_03_capsule, center=<10.4865 0.3793 0.51>, rotation=<rot 0 90.3808 -87.0606>}]
 dyn_scarf_r_04 [{shapeType=Capsule, length=13, radius=3, shapeName=dyn_scarf_r_04_capsule, center=<9.8162 0.1809 0.1012>, rotation=<rot 0 -91.1931 -81.7313>}]
 dyn_scarf_r_05 [{shapeType=Capsule, length=13, radius=3, shapeName=dyn_scarf_r_05_capsule, center=<10.2957 0.0029 -1.1568>, rotation=<rot 0 97.514 -84.3837>}]
 dyn_scarf_r_base [{shapeType=Capsule, length=0, radius=5, shapeName=dyn_scarf_r_base_capsule, rotation=<rot 0 -90 0>}]
 head [{shapeType=Capsule, length=1.5, radius=17, shapeName=head_capsule, center=<12.9301 0.9557 0>}]
 pelvis [{shapeType=Capsule, length=18.4621, radius=15.3153, center=<-4.4285 -2.1256 -0.1986>, rotation=<rot 2.8068 -90 -0.212>}]
 spine_01 [{shapeType=Capsule, length=16.8035, radius=12.3979, center=<7.6511 -1.783 0.0942>, rotation=<rot -2.9397 -90 0.0929>}]
 spine_03 [{shapeType=Capsule, length=16.3891, radius=12.2138, center=<1.8961 -1.2528 0.0582>, rotation=<rot -2.3684 -90 0.2751>}]
 spine_04 [{shapeType=Capsule, length=16.2983, radius=12.0624, center=<1.6804 -0.6542 -0.1104>, rotation=<rot 2.0309 -90 -0.3755>}]
 spine_05 [{shapeType=Capsule, length=25.9246, radius=8.3035, center=<5.0585 -0.7464 -0.0282>, rotation=<rot -2.4753 -90 -2.0518>}]
 upperarm_l [{shapeType=Capsule, length=25.7561, radius=7.2638, center=<15.5735 -0.5444 -1.0404>, rotation=<rot 0 101.2207 -86.8193>}]
 upperarm_r [{shapeType=Capsule, length=25.9898, radius=7.4507, center=<-13.5419 0.4597 0.9725>, rotation=<rot 0 89.3516 -88.5205>}]
}
constraints [
 {bone1Name=spine_01, bone2Name=pelvis, swing1LimitDegrees=45, swing2LimitDegrees=45, twistLimitDegrees=45}
 {bone1Name=spine_03, bone2Name=spine_01, swing1LimitDegrees=45, swing2LimitDegrees=45, twistLimitDegrees=45}
 {bone1Name=spine_04, bone2Name=spine_03, swing1LimitDegrees=45, swing2LimitDegrees=45, twistLimitDegrees=45}
 {bone1Name=spine_05, bone2Name=spine_04, swing1LimitDegrees=45, swing2LimitDegrees=45, twistLimitDegrees=45}
 {bone1Name=upperarm_l, bone2Name=spine_05, swing1LimitDegrees=45, swing2LimitDegrees=45, twistLimitDegrees=45}
 {bone1Name=upperarm_r, bone2Name=spine_05, swing1LimitDegrees=45, swing2LimitDegrees=45, twistLimitDegrees=45}
 {bone1Name=dyn_scarf_r_05, bone2Name=dyn_scarf_r_04, swing1LimitDegrees=15, swing2LimitDegrees=45, twistLimitDegrees=5}
 {bone1Name=dyn_scarf_r_04, bone2Name=dyn_scarf_r_03, swing1LimitDegrees=15, swing2LimitDegrees=45, twistLimitDegrees=5}
 {bone1Name=dyn_scarf_r_03, bone2Name=dyn_scarf_r_02, swing1LimitDegrees=15, swing2LimitDegrees=45, twistLimitDegrees=5}
 {bone1Name=dyn_scarf_r_02, bone2Name=dyn_scarf_r_01, swing1LimitDegrees=15, swing2LimitDegrees=15, twistLimitDegrees=5}
 {bone1Name=dyn_scarf_r_01, bone2Name=dyn_scarf_r_base, swing1LimitDegrees=10, swing2LimitDegrees=10, twistLimitDegrees=5}
 {bone1Name=dyn_scarf_l_05, bone2Name=dyn_scarf_l_04, swing1LimitDegrees=15, swing2LimitDegrees=45, twistLimitDegrees=5}
 {bone1Name=dyn_scarf_l_04, bone2Name=dyn_scarf_l_03, swing1LimitDegrees=15, swing2LimitDegrees=45, twistLimitDegrees=5}
 {bone1Name=dyn_scarf_l_03, bone2Name=dyn_scarf_l_02, swing1LimitDegrees=15, swing2LimitDegrees=45, twistLimitDegrees=5}
 {bone1Name=dyn_scarf_l_02, bone2Name=dyn_scarf_l_01, swing1LimitDegrees=15, swing2LimitDegrees=15, twistLimitDegrees=5}
 {bone1Name=dyn_scarf_l_01, bone2Name=dyn_scarf_l_base, swing1LimitDegrees=10, swing2LimitDegrees=10, twistLimitDegrees=5}
 {
  bone1Name=dyn_scarf_l_base
  bone2Name=spine_05
  swing1LimitDegrees=45
  swing1Motion=Locked
  swing2LimitDegrees=45
  swing2Motion=Locked
  twistLimitDegrees=45
  twistMotion=Locked
 }
 {
  bone1Name=dyn_scarf_r_base
  bone2Name=spine_05
  swing1LimitDegrees=45
  swing1Motion=Locked
  swing2LimitDegrees=45
  swing2Motion=Locked
  twistLimitDegrees=45
  twistMotion=Locked
 }
]
