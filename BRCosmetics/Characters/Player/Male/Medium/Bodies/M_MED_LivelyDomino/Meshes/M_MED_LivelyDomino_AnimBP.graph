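# Animation Blueprint for the LivelyDomino male player body, adding secondary motion to the scarf and upper-arm sleeves over a cached control rig pose.
# Without backbling, the scarf uses RigidBody simulation, switching setups for skydiving with variable-driven gravity and bone offsets.
# With backbling, weighted spine and clavicle constraints plus bone offsets position the scarf, with spring AnimDynamics on its lower bones.
# Spring AnimDynamics then moves both sleeves.
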AnimBlueprint /BRCosmetics/Characters/Player/Male/Medium/Bodies/M_MED_LivelyDomino/Meshes/M_MED_LivelyDomino_AnimBP
target_skeleton=&/M_MED_LivelyDomino_Skeleton
graphs [
 AnimGraph {nodes [
   RigidBody!(ComponentPose, Alpha, OverrideWorldGravity:Vector=Get GravityOverride_tessel() -> (GravityOverride_tessel:Vector), ExternalForce) -> (Pose) {
    bClampLinearTranslationLimitToRefPose=true
    bOverrideWorldGravity=true
    overridePhysicsAsset=&/M_MED_LivelyDomino_WearingBackblingPhysics
    simulationSpace=BaseBoneSpace
    alphaBoolBlend {blendOption=Linear}
    baseBoneRef {boneName=gravity}
    componentAppliedLinearAccClamp=<3000 3000 3000>
    componentLinearAccScale=<0.05 0.05 0.05>
    componentLinearVelScale=<0.05 0.05 0.05>
    overrideWorldGravity {z=-980}
    simSpaceSettings {
     maxAngularAcceleration=50
     maxAngularVelocity=1
     maxLinearAcceleration=300
     maxLinearVelocity=150
     velocityScaleZ=0
     externalLinearDragV=<0.2 0.2>
    }
   }
   n1 = "Constraint - dyn_scarf_l_base (PP)"!(
    ComponentPose=Local To Component(LocalPose=Use cached pose 'control rig pose'())
    Alpha=1
    ConstraintWeights_0=0.9
    ConstraintWeights_1:float=0.1
   ) -> (Pose) {
    alphaBoolBlend {blendOption=Linear}
    boneToModify {boneName=dyn_scarf_l_base}
    constraintSetup [{targetBone {boneName=spine_05}}, {targetBone {boneName=clavicle_l}}]
    constraintWeights [1,1]
   }
   n2 = "Constraint - dyn_scarf_r_base (PP)"!(ComponentPose=@n1, Alpha=1, ConstraintWeights_0=0.9, ConstraintWeights_1:float=0.1) -> (Pose) {
    alphaBoolBlend {blendOption=Linear}
    boneToModify {boneName=dyn_scarf_r_base}
    constraintSetup [{targetBone {boneName=spine_05}}, {targetBone {boneName=clavicle_r}}]
    constraintWeights [1,1]
   }
   n3 = "Constraint - dyn_scarf_l_01 (PP)"!(ComponentPose=@n2, Alpha=1, ConstraintWeights_0=0.8, ConstraintWeights_1:float=0.2) -> (Pose) {
    alphaBoolBlend {blendOption=Linear}
    boneToModify {boneName=dyn_scarf_l_01}
    constraintSetup [{targetBone {boneName=spine_05}}, {targetBone {boneName=clavicle_l}}]
    constraintWeights [1,1]
   }
   n4 = "Constraint - dyn_scarf_r_01 (PP)"!(ComponentPose=@n3, Alpha=1, ConstraintWeights_0=0.8, ConstraintWeights_1:float=0.2) -> (Pose) {
    alphaBoolBlend {blendOption=Linear}
    boneToModify {boneName=dyn_scarf_r_01}
    constraintSetup [{targetBone {boneName=spine_05}}, {targetBone {boneName=clavicle_r}}]
    constraintWeights [1,1]
   }
   n5 = "Transform (Modify) Bone - Bone: dyn_scarf_l_base"(ComponentPose=@n4, Alpha=1, Translation="0,0,0", Rotation="-15,0,0", Scale="1,1,1") {
    rotationMode=BMM_Additive
    rotationSpace=BCS_BoneSpace
    translationSpace=BCS_BoneSpace
    alphaBoolBlend {blendOption=Linear}
    alphaScaleBiasClamp {interpSpeedDecreasing=2, interpSpeedIncreasing=2}
    boneToModify {boneName=dyn_scarf_l_base}
    rotation {pitch=-15}
   }
   n6 = "Transform (Modify) Bone - Bone: dyn_scarf_r_base"(ComponentPose=@n5, Alpha=1, Translation="0,0,0", Rotation="-15,0,0", Scale="1,1,1") {
    rotationMode=BMM_Additive
    rotationSpace=BCS_BoneSpace
    translationSpace=BCS_BoneSpace
    alphaBoolBlend {blendOption=Linear}
    alphaScaleBiasClamp {interpSpeedDecreasing=2, interpSpeedIncreasing=2}
    boneToModify {boneName=dyn_scarf_r_base}
    rotation {pitch=-15}
   }
   n7 = "Constraint - dyn_scarf_l_02 (PP)"!(ComponentPose=@n6, Alpha=1, ConstraintWeights_0=0.5, ConstraintWeights_1:float=0.5) -> (Pose) {
    alphaBoolBlend {blendOption=Linear}
    boneToModify {boneName=dyn_scarf_l_02}
    constraintSetup [{targetBone {boneName=spine_04}}, {targetBone {boneName=spine_03}}]
    constraintWeights [1,1]
   }
   n8 = "Constraint - dyn_scarf_r_02 (PP)"!(ComponentPose=@n7, Alpha=1, ConstraintWeights_0=0.5, ConstraintWeights_1:float=0.5) -> (Pose) {
    alphaBoolBlend {blendOption=Linear}
    boneToModify {boneName=dyn_scarf_r_02}
    constraintSetup [{targetBone {boneName=spine_04}}, {targetBone {boneName=spine_03}}]
    constraintWeights [1,1]
   }
   n9 = "Transform (Modify) Bone - Bone: dyn_scarf_l_02"(ComponentPose=@n8, Alpha=1, Translation="0,0,2", Rotation="0,0,0", Scale="1,1,1") {
    rotationMode=BMM_Additive
    rotationSpace=BCS_BoneSpace
    translationMode=BMM_Additive
    translationSpace=BCS_BoneSpace
    alphaBoolBlend {blendOption=Linear}
    alphaScaleBiasClamp {interpSpeedDecreasing=2, interpSpeedIncreasing=2}
    boneToModify {boneName=dyn_scarf_l_02}
    translation {z=2}
   }
   n10 = "Transform (Modify) Bone - Bone: dyn_scarf_r_02"(ComponentPose=@n9, Alpha=1, Translation="0,0,1.5", Rotation="0,0,0", Scale="1,1,1") {
    rotationMode=BMM_Additive
    rotationSpace=BCS_BoneSpace
    translationMode=BMM_Additive
    translationSpace=BCS_BoneSpace
    alphaBoolBlend {blendOption=Linear}
    alphaScaleBiasClamp {interpSpeedDecreasing=2, interpSpeedIncreasing=2}
    boneToModify {boneName=dyn_scarf_r_02}
    translation {z=1.5}
   }
   n11 = "Constraint - dyn_scarf_l_03 (PP)"!(ComponentPose=@n10, Alpha=1, ConstraintWeights_0=0.5, ConstraintWeights_1:float=0.5) -> (Pose) {
    alphaBoolBlend {blendOption=Linear}
    boneToModify {boneName=dyn_scarf_l_03}
    constraintSetup [{targetBone {boneName=spine_04}}, {targetBone {boneName=spine_03}}]
    constraintWeights [1,1]
   }
   n12 = "Constraint - dyn_scarf_r_03 (PP)"!(ComponentPose=@n11, Alpha=1, ConstraintWeights_0=0.5, ConstraintWeights_1:float=0.5) -> (Pose) {
    alphaBoolBlend {blendOption=Linear}
    boneToModify {boneName=dyn_scarf_r_03}
    constraintSetup [{targetBone {boneName=spine_04}}, {targetBone {boneName=spine_03}}]
    constraintWeights [1,1]
   }
   n13 = "Transform (Modify) Bone - Bone: dyn_scarf_l_03"(ComponentPose=@n12, Alpha=1, Translation="0,0,0", Rotation="-10,0,0", Scale="1,1,1") {
    rotationMode=BMM_Additive
    rotationSpace=BCS_BoneSpace
    translationMode=BMM_Additive
    translationSpace=BCS_BoneSpace
    alphaBoolBlend {blendOption=Linear}
    alphaScaleBiasClamp {interpSpeedDecreasing=2, interpSpeedIncreasing=2}
    boneToModify {boneName=dyn_scarf_l_03}
    rotation {pitch=-10}
   }
   n14 = "Transform (Modify) Bone - Bone: dyn_scarf_r_03"(ComponentPose=@n13, Alpha=1, Translation="0,0,0", Rotation="-10,0,0", Scale="1,1,1") {
    rotationMode=BMM_Additive
    rotationSpace=BCS_BoneSpace
    translationMode=BMM_Additive
    translationSpace=BCS_BoneSpace
    alphaBoolBlend {blendOption=Linear}
    alphaScaleBiasClamp {interpSpeedDecreasing=2, interpSpeedIncreasing=2}
    boneToModify {boneName=dyn_scarf_r_03}
    rotation {pitch=-10}
   }
   n15 = "Transform (Modify) Bone - Bone: dyn_scarf_l_01"(ComponentPose=@n14, Alpha=1, Translation="0,0,0", Rotation="10,0,0", Scale="1,1,1") {
    rotationMode=BMM_Replace
    rotationSpace=BCS_ParentBoneSpace
    translationSpace=BCS_BoneSpace
    alphaBoolBlend {blendOption=Linear}
    alphaScaleBiasClamp {interpSpeedDecreasing=2, interpSpeedIncreasing=2}
    boneToModify {boneName=dyn_scarf_l_01}
    rotation {pitch=10}
   }
   n16 = "Transform (Modify) Bone - Bone: dyn_scarf_r_01"(ComponentPose=@n15, Alpha=1, Translation="0,0,0", Rotation="10,0,0", Scale="1,1,1") {
    rotationMode=BMM_Replace
    rotationSpace=BCS_ParentBoneSpace
    translationSpace=BCS_BoneSpace
    alphaBoolBlend {blendOption=Linear}
    alphaScaleBiasClamp {interpSpeedDecreasing=2, interpSpeedIncreasing=2}
    boneToModify {boneName=dyn_scarf_r_01}
    rotation {pitch=10}
   }
   n17 = "Transform (Modify) Bone - Bone: dyn_scarf_l_01"(ComponentPose=@n16, Alpha=1, Translation="0,0,0", Rotation="30,0,0", Scale="1,1,1") {
    rotationMode=BMM_Replace
    rotationSpace=BCS_ParentBoneSpace
    translationSpace=BCS_BoneSpace
    alphaBoolBlend {blendOption=Linear}
    alphaScaleBiasClamp {interpSpeedDecreasing=2, interpSpeedIncreasing=2}
    boneToModify {boneName=dyn_scarf_l_01}
    rotation {pitch=30}
   }
   n18 = "Transform (Modify) Bone - Bone: dyn_scarf_r_01"(ComponentPose=@n17, Alpha=1, Translation="0,0,0", Rotation="30,0,0", Scale="1,1,1") {
    rotationMode=BMM_Replace
    rotationSpace=BCS_ParentBoneSpace
    translationSpace=BCS_BoneSpace
    alphaBoolBlend {blendOption=Linear}
    alphaScaleBiasClamp {interpSpeedDecreasing=2, interpSpeedIncreasing=2}
    boneToModify {boneName=dyn_scarf_r_01}
    rotation {pitch=30}
   }
   n19 = "Transform (Modify) Bone - Bone: dyn_scarf_l_03"(ComponentPose=@n18, Alpha=1, Translation="0,0,6", Rotation="-10,0,0", Scale="1,1,1") {
    rotationMode=BMM_Replace
    rotationSpace=BCS_ParentBoneSpace
    translationMode=BMM_Additive
    translationSpace=BCS_BoneSpace
    alphaBoolBlend {blendOption=Linear}
    alphaScaleBiasClamp {interpSpeedDecreasing=2, interpSpeedIncreasing=2}
    boneToModify {boneName=dyn_scarf_l_03}
    rotation {pitch=-10}
    translation {z=6}
   }
   n20 = "Transform (Modify) Bone - Bone: dyn_scarf_r_03"(ComponentPose=@n19, Alpha=1, Translation="0,0,6", Rotation="-10,0,0", Scale="1,1,1") {
    rotationMode=BMM_Replace
    rotationSpace=BCS_ParentBoneSpace
    translationMode=BMM_Additive
    translationSpace=BCS_BoneSpace
    alphaBoolBlend {blendOption=Linear}
    alphaScaleBiasClamp {interpSpeedDecreasing=2, interpSpeedIncreasing=2}
    boneToModify {boneName=dyn_scarf_r_03}
    rotation {pitch=-10}
    translation {z=6}
   }
   n21 = Component To Local(ComponentPose=@n20)
   n22 = "Constraint - dyn_scarf_l_base (PP)"(ComponentPose=Local To Component(LocalPose=Use cached pose 'control rig pose'()), Alpha=1, ConstraintWeights_0=0.9, ConstraintWeights_1=0.1) {
    alphaBoolBlend {blendOption=Linear}
    boneToModify {boneName=dyn_scarf_l_base}
    constraintSetup [{targetBone {boneName=spine_05}}, {targetBone {boneName=clavicle_l}}]
    constraintWeights [1,1]
   }
   n23 = "Constraint - dyn_scarf_r_base (PP)"(ComponentPose=@n22, Alpha=1, ConstraintWeights_0=0.9, ConstraintWeights_1=0.1) {
    alphaBoolBlend {blendOption=Linear}
    boneToModify {boneName=dyn_scarf_r_base}
    constraintSetup [{targetBone {boneName=spine_05}}, {targetBone {boneName=clavicle_r}}]
    constraintWeights [1,1]
   }
   n24 = "Transform (Modify) Bone - Bone: dyn_scarf_l_base"(ComponentPose=@n23, Alpha=1, Translation="0,0,0", Rotation="-15.000,0.001,0.081", Scale="1,1,1") {
    rotationMode=BMM_Additive
    rotationSpace=BCS_BoneSpace
    translationSpace=BCS_BoneSpace
    alphaBoolBlend {blendOption=Linear}
    alphaScaleBiasClamp {interpSpeedDecreasing=2, interpSpeedIncreasing=2}
    boneToModify {boneName=dyn_scarf_l_base}
    rotation=<rot -15 0.001 0.081>
   }
   n25 = "Transform (Modify) Bone - Bone: dyn_scarf_r_base"(ComponentPose=@n24, Alpha=1, Translation="0,0,0", Rotation="-15,0,0", Scale="1,1,1") {
    rotationMode=BMM_Additive
    rotationSpace=BCS_BoneSpace
    translationSpace=BCS_BoneSpace
    alphaBoolBlend {blendOption=Linear}
    alphaScaleBiasClamp {interpSpeedDecreasing=2, interpSpeedIncreasing=2}
    boneToModify {boneName=dyn_scarf_r_base}
    rotation {pitch=-15}
   }
   n26 = "Constraint - dyn_scarf_l_01 (PP)"(ComponentPose=@n25, Alpha=1, ConstraintWeights_0=0.8, ConstraintWeights_1=0.2) {
    alphaBoolBlend {blendOption=Linear}
    boneToModify {boneName=dyn_scarf_l_01}
    constraintSetup [{targetBone {boneName=spine_05}}, {targetBone {boneName=clavicle_l}}]
    constraintWeights [1,1]
   }
   n27 = "Constraint - dyn_scarf_r_01 (PP)"(ComponentPose=@n26, Alpha=1, ConstraintWeights_0=0.8, ConstraintWeights_1=0.2) {
    alphaBoolBlend {blendOption=Linear}
    boneToModify {boneName=dyn_scarf_r_01}
    constraintSetup [{targetBone {boneName=spine_05}}, {targetBone {boneName=clavicle_r}}]
    constraintWeights [1,1]
   }
   n28 = "Transform (Modify) Bone - Bone: dyn_scarf_l_01"(ComponentPose=@n27, Alpha=1, Translation="0,0,-2", Rotation="5.026,0.673,0.341", Scale="1,1,1") {
    rotationMode=BMM_Additive
    rotationSpace=BCS_BoneSpace
    translationMode=BMM_Additive
    translationSpace=BCS_BoneSpace
    alphaBoolBlend {blendOption=Linear}
    alphaScaleBiasClamp {interpSpeedDecreasing=2, interpSpeedIncreasing=2}
    boneToModify {boneName=dyn_scarf_l_01}
    rotation=<rot 5.026 0.673 0.341>
    translation {z=-2}
   }
   n29 = "Transform (Modify) Bone - Bone: dyn_scarf_r_01"(ComponentPose=@n28, Alpha=1, Translation="0,0,-2", Rotation="-13.000,-0.041,-0.007", Scale="1,1,1") {
    rotationMode=BMM_Additive
    rotationSpace=BCS_BoneSpace
    translationMode=BMM_Additive
    translationSpace=BCS_BoneSpace
    alphaBoolBlend {blendOption=Linear}
    alphaScaleBiasClamp {interpSpeedDecreasing=2, interpSpeedIncreasing=2}
    boneToModify {boneName=dyn_scarf_r_01}
    rotation=<rot -13 -0.041 -0.007>
    translation {z=-2}
   }
   n30 = "Constraint - dyn_scarf_l_02 (PP)"(ComponentPose=@n29, Alpha=1, ConstraintWeights_0=0.5, ConstraintWeights_1=0.5) {
    alphaBoolBlend {blendOption=Linear}
    boneToModify {boneName=dyn_scarf_l_02}
    constraintSetup [{targetBone {boneName=spine_04}}, {targetBone {boneName=spine_03}}]
    constraintWeights [1,1]
   }
   n31 = "Constraint - dyn_scarf_r_02 (PP)"(ComponentPose=@n30, Alpha=1, ConstraintWeights_0=0.5, ConstraintWeights_1=0.5) {
    alphaBoolBlend {blendOption=Linear}
    boneToModify {boneName=dyn_scarf_r_02}
    constraintSetup [{targetBone {boneName=spine_04}}, {targetBone {boneName=spine_03}}]
    constraintWeights [1,1]
   }
   n32 = "Constraint - dyn_scarf_l_03 (PP)"(ComponentPose=@n31, Alpha=1, ConstraintWeights_0=0.5, ConstraintWeights_1=0.5) {
    alphaBoolBlend {blendOption=Linear}
    boneToModify {boneName=dyn_scarf_l_03}
    constraintSetup [{targetBone {boneName=spine_04}}, {targetBone {boneName=spine_03}}]
    constraintWeights [1,1]
   }
   n33 = "Constraint - dyn_scarf_r_03 (PP)"(ComponentPose=@n32, Alpha=1, ConstraintWeights_0=0.5, ConstraintWeights_1=0.5) {
    alphaBoolBlend {blendOption=Linear}
    boneToModify {boneName=dyn_scarf_r_03}
    constraintSetup [{targetBone {boneName=spine_04}}, {targetBone {boneName=spine_03}}]
    constraintWeights [1,1]
   }
   n34 = "Transform (Modify) Bone - Bone: dyn_scarf_l_02"(ComponentPose=@n33, Alpha=1, Translation="0,0,-1", Rotation="5.000,0.042,-0.034", Scale="1,1,1") {
    rotationMode=BMM_Additive
    rotationSpace=BCS_BoneSpace
    translationMode=BMM_Additive
    translationSpace=BCS_BoneSpace
    alphaBoolBlend {blendOption=Linear}
    alphaScaleBiasClamp {interpSpeedDecreasing=2, interpSpeedIncreasing=2}
    boneToModify {boneName=dyn_scarf_l_02}
    rotation=<rot 5 0.042 -0.034>
    translation {z=-1}
   }
   n35 = "Transform (Modify) Bone - Bone: dyn_scarf_r_02"(ComponentPose=@n34, Alpha=1, Translation="0,0,-1", Rotation="0,0.05,0.004", Scale="1,1,1") {
    rotationMode=BMM_Additive
    rotationSpace=BCS_BoneSpace
    translationMode=BMM_Additive
    translationSpace=BCS_BoneSpace
    alphaBoolBlend {blendOption=Linear}
    alphaScaleBiasClamp {interpSpeedDecreasing=2, interpSpeedIncreasing=2}
    boneToModify {boneName=dyn_scarf_r_02}
    rotation {roll=0.004, yaw=0.05}
    translation {z=-1}
   }
   n36 = "Transform (Modify) Bone - Bone: dyn_scarf_l_03"(ComponentPose=@n35, Alpha=1, Translation="-1.120,-0.131,2.921", Rotation="-20,0,0", Scale="1,1,1") {
    rotationMode=BMM_Additive
    rotationSpace=BCS_BoneSpace
    translationMode=BMM_Additive
    translationSpace=BCS_BoneSpace
    alphaBoolBlend {blendOption=Linear}
    alphaScaleBiasClamp {interpSpeedDecreasing=2, interpSpeedIncreasing=2}
    boneToModify {boneName=dyn_scarf_l_03}
    rotation {pitch=-20}
    translation=<-1.12 -0.131 2.921>
   }
   n37 = "Transform (Modify) Bone - Bone: dyn_scarf_r_03"(ComponentPose=@n36, Alpha=1, Translation="0.821,1.651,2.169", Rotation="-20,0,0", Scale="1,1,1") {
    rotationMode=BMM_Additive
    rotationSpace=BCS_BoneSpace
    translationMode=BMM_Additive
    translationSpace=BCS_BoneSpace
    alphaBoolBlend {blendOption=Linear}
    alphaScaleBiasClamp {interpSpeedDecreasing=2, interpSpeedIncreasing=2}
    boneToModify {boneName=dyn_scarf_r_03}
    rotation {pitch=-20}
    translation=<0.821 1.651 2.169>
   }
   n38 = "Constraint - dyn_scarf_l_04 (P)"(ComponentPose=@n37, Alpha=1, ConstraintWeights_0=1) {
    alphaBoolBlend {blendOption=Linear}
    boneToModify {boneName=dyn_scarf_l_04}
    constraintSetup [{targetBone {boneName=spine_03}}]
    constraintWeights [1]
   }
   n39 = "Constraint - dyn_scarf_r_04 (P)"(ComponentPose=@n38, Alpha=1, ConstraintWeights_0=1) {
    alphaBoolBlend {blendOption=Linear}
    boneToModify {boneName=dyn_scarf_r_04}
    constraintSetup [{targetBone {boneName=spine_03}}]
    constraintWeights [1]
   }
   n40 = Anim Dynamics - Bone: dyn_scarf_l_04!(ComponentPose=@n39, Alpha=0.8, GravityOverride:Vector="0,0,0", ExternalForce="0,0,0") -> (Pose) {
    angularDampingOverride=1
    angularSpringConstant=50
    bAngularSpring=true
    bOverrideAngularDamping=true
    bUseGravityOverride=true
    simulationSpace=RootRelative
    alphaBoolBlend {blendOption=Linear}
    boundBone {boneName=dyn_scarf_l_04}
    componentAppliedLinearAccClamp=<1000 1000 1000>
    componentLinearAccScale=<0.2 0.2 0.2>
    componentLinearVelScale=<0.2 0.2 0.2>
    physicsBodyDefinitions [{
      boundBone {boneName=dyn_scarf_l_04}
      boxExtents=<10 10 10>
      constraintSetup {angularLimitsMax=<1 25 15>, angularLimitsMin=<-1 -7 -15>, angularTarget=<1 0 0>}
      localJointOffset=<2 0 0>
     }]
   }
   n41 = Anim Dynamics - Bone: dyn_scarf_r_04!(ComponentPose=@n40, Alpha=1, GravityOverride:Vector="0,0,0", ExternalForce="0,0,0") -> (Pose) {
    angularDampingOverride=1
    angularSpringConstant=80
    bAngularSpring=true
    bOverrideAngularDamping=true
    bUseGravityOverride=true
    simulationSpace=RootRelative
    alphaBoolBlend {blendOption=Linear}
    boundBone {boneName=dyn_scarf_r_04}
    componentAppliedLinearAccClamp=<1000 1000 1000>
    componentLinearAccScale=<0.2 0.2 0.2>
    componentLinearVelScale=<0.2 0.2 0.2>
    physicsBodyDefinitions [{
      boundBone {boneName=dyn_scarf_r_04}
      boxExtents=<10 10 10>
      constraintSetup {angularLimitsMax=<1 25 15>, angularLimitsMin=<-1 -7 -15>, angularTarget=<1 0 0>}
      localJointOffset=<2 0 0>
     }]
   }
   n42 = Anim Dynamics - Bone: dyn_scarf_l_05!(ComponentPose=@n41, Alpha=0.8, GravityOverride:Vector="0,0,0", ExternalForce="0,0,0") -> (Pose) {
    angularDampingOverride=1
    angularSpringConstant=80
    bAngularSpring=true
    bOverrideAngularDamping=true
    bUseGravityOverride=true
    simulationSpace=RootRelative
    alphaBoolBlend {blendOption=Linear}
    boundBone {boneName=dyn_scarf_l_05}
    componentAppliedLinearAccClamp=<1000 1000 1000>
    componentLinearAccScale=<0.2 0.2 0.2>
    componentLinearVelScale=<0.2 0.2 0.2>
    physicsBodyDefinitions [{
      boundBone {boneName=dyn_scarf_l_05}
      boxExtents=<10 10 10>
      constraintSetup {angularLimitsMax=<1 25 15>, angularLimitsMin=<-1 -7 -15>, angularTarget=<1 0 0>}
      localJointOffset=<2 0 0>
     }]
   }
   n43 = Anim Dynamics - Bone: dyn_scarf_r_05!(ComponentPose=@n42, Alpha=1, GravityOverride:Vector="0,0,0", ExternalForce="0,0,0") -> (Pose) {
    angularDampingOverride=1
    angularSpringConstant=50
    bAngularSpring=true
    bOverrideAngularDamping=true
    bUseGravityOverride=true
    simulationSpace=RootRelative
    alphaBoolBlend {blendOption=Linear}
    boundBone {boneName=dyn_scarf_r_05}
    componentAppliedLinearAccClamp=<1000 1000 1000>
    componentLinearAccScale=<0.2 0.2 0.2>
    componentLinearVelScale=<0.2 0.2 0.2>
    physicsBodyDefinitions [{
      boundBone {boneName=dyn_scarf_r_05}
      boxExtents=<10 10 10>
      constraintSetup {angularLimitsMax=<1 25 15>, angularLimitsMin=<-1 -7 -15>, angularTarget=<1 0 0>}
      localJointOffset=<2 0 0>
     }]
   }
   n44 = "Transform (Modify) Bone - Bone: dyn_scarf_l_04"(ComponentPose=@n43, Alpha=1, Translation="0,0,0", Rotation="0,0,0", Scale="1,1,1") {
    rotationMode=BMM_Additive
    rotationSpace=BCS_BoneSpace
    translationMode=BMM_Additive
    translationSpace=BCS_BoneSpace
    alphaBoolBlend {blendOption=Linear}
    alphaScaleBiasClamp {interpSpeedDecreasing=2, interpSpeedIncreasing=2}
    boneToModify {boneName=dyn_scarf_l_04}
   }
   n45 = "Transform (Modify) Bone - Bone: dyn_scarf_r_04"(ComponentPose=@n44, Alpha=1, Translation="0,0,-1.3", Rotation="0,0,0", Scale="1,1,1") {
    rotationMode=BMM_Additive
    rotationSpace=BCS_BoneSpace
    translationMode=BMM_Additive
    translationSpace=BCS_BoneSpace
    alphaBoolBlend {blendOption=Linear}
    alphaScaleBiasClamp {interpSpeedDecreasing=2, interpSpeedIncreasing=2}
    boneToModify {boneName=dyn_scarf_r_04}
    translation {z=-1.3}
   }
   n46 = "Transform (Modify) Bone - Bone: dyn_scarf_l_03"(ComponentPose=@n45, Alpha=1, Translation="0,0,-1", Rotation="0,0,0", Scale="1,1,1") {
    rotationMode=BMM_Additive
    rotationSpace=BCS_BoneSpace
    translationMode=BMM_Additive
    translationSpace=BCS_BoneSpace
    alphaBoolBlend {blendOption=Linear}
    alphaScaleBiasClamp {bInterpResult=true, interpSpeedDecreasing=25, interpSpeedIncreasing=25}
    boneToModify {boneName=dyn_scarf_l_03}
    rotation=<rot 0 0 0>
    translation {z=-1}
   }
   n47 = "Transform (Modify) Bone - Bone: dyn_scarf_r_03"(ComponentPose=@n46, Alpha=1, Translation="0,0,-1", Rotation="0,0,0", Scale="1,1,1") {
    rotationMode=BMM_Additive
    rotationSpace=BCS_BoneSpace
    scaleSpace=BCS_BoneSpace
    translationMode=BMM_Additive
    translationSpace=BCS_BoneSpace
    alphaBoolBlend {blendOption=Linear}
    alphaScaleBiasClamp {bInterpResult=true, interpSpeedDecreasing=25, interpSpeedIncreasing=25}
    boneToModify {boneName=dyn_scarf_r_03}
    rotation=<rot 0 0 0>
    translation {z=-1}
   }
   n48 = "Transform (Modify) Bone - Bone: dyn_scarf_l_04"(ComponentPose=@n47, Alpha=1, Translation="0,0,-4", Rotation="-20,4.475,1.137", Scale="1,1,1") {
    rotationMode=BMM_Replace
    rotationSpace=BCS_ParentBoneSpace
    translationMode=BMM_Additive
    translationSpace=BCS_BoneSpace
    alphaBoolBlend {blendOption=Linear}
    alphaScaleBiasClamp {bInterpResult=true, interpSpeedDecreasing=25, interpSpeedIncreasing=25}
    boneToModify {boneName=dyn_scarf_l_04}
    rotation=<rot -20 4.475 1.137>
    translation {z=-4}
   }
   n49 = "Transform (Modify) Bone - Bone: dyn_scarf_r_04"(ComponentPose=@n48, Alpha=1, Translation="0,0,-2", Rotation="-10,-1.463,0.561", Scale="1,1,1") {
    rotationMode=BMM_Replace
    rotationSpace=BCS_ParentBoneSpace
    translationMode=BMM_Additive
    translationSpace=BCS_BoneSpace
    alphaBoolBlend {blendOption=Linear}
    alphaScaleBiasClamp {bInterpResult=true, interpSpeedDecreasing=25, interpSpeedIncreasing=25}
    boneToModify {boneName=dyn_scarf_r_04}
    rotation=<rot -10 -1.463 0.561>
    translation {z=-2}
   }
   n50 = "Transform (Modify) Bone - Bone: dyn_scarf_l_05"(ComponentPose=@n49, Alpha=1, Translation="0,0,-4", Rotation="0,0,0", Scale="1,1,1") {
    rotationMode=BMM_Replace
    rotationSpace=BCS_ParentBoneSpace
    translationMode=BMM_Additive
    translationSpace=BCS_BoneSpace
    alphaBoolBlend {blendOption=Linear}
    alphaScaleBiasClamp {bInterpResult=true, interpSpeedDecreasing=25, interpSpeedIncreasing=25}
    boneToModify {boneName=dyn_scarf_l_05}
    translation {z=-4}
   }
   n51 = "Transform (Modify) Bone - Bone: dyn_scarf_r_05"(ComponentPose=@n50, Alpha=1, Translation="0,0,-2", Rotation="0,0,0", Scale="1,1,1") {
    rotationMode=BMM_Replace
    rotationSpace=BCS_ParentBoneSpace
    translationMode=BMM_Additive
    translationSpace=BCS_BoneSpace
    alphaBoolBlend {blendOption=Linear}
    alphaScaleBiasClamp {bInterpResult=true, interpSpeedDecreasing=25, interpSpeedIncreasing=25}
    boneToModify {boneName=dyn_scarf_r_05}
    translation {z=-2}
   }
   n52 = Component To Local(ComponentPose=@n51)
   n53 = Get Is Wearing Backbling()
   n54 = RigidBody(ComponentPose=Local To Component(LocalPose=Use cached pose 'control rig pose'()), OverrideWorldGravity=Get GravityOverride_tessel()) {
    bClampLinearTranslationLimitToRefPose=true
    bOverrideWorldGravity=true
    overridePhysicsAsset=&/M_MED_LivelyDomino_Physics
    simulationSpace=BaseBoneSpace
    alphaBoolBlend {blendOption=Linear}
    baseBoneRef {boneName=gravity}
    componentAppliedLinearAccClamp=<3000 3000 3000>
    componentLinearAccScale {x=0.1}
    componentLinearVelScale {x=0.1}
    overrideWorldGravity {z=-980}
    simSpaceSettings {
     maxAngularAcceleration=1000
     maxAngularVelocity=3
     maxLinearAcceleration=700
     maxLinearVelocity=500
     velocityScaleZ=0
     worldAlpha=1
     externalLinearDragV=<0.5 0.5>
    }
   }
   n55 = "Transform (Modify) Bone - Bone: dyn_scarf_root"(ComponentPose=@n54, Alpha=Get Is Skydiving Alpha() -> (Is Skydiving Alpha:double), Translation="0,0,0", Rotation="-20,0,0", Scale="1,1,1") {
    rotationMode=BMM_Additive
    rotationSpace=BCS_BoneSpace
    translationSpace=BCS_BoneSpace
    alphaBoolBlend {blendOption=Linear}
    alphaScaleBiasClamp {interpSpeedDecreasing=2, interpSpeedIncreasing=2}
    boneToModify {boneName=dyn_scarf_root}
    rotation {pitch=-20}
   }
   n56 = "Transform (Modify) Bone - Bone: dyn_scarf_l_base"(
    ComponentPose=Local To Component(LocalPose=Use cached pose 'control rig pose'())
    Alpha=Get Tassle 01 Alpha() -> (Tassle 01 Alpha:double)
    Translation="0,0,0"
    Rotation=Get TesselRotator_01() -> (TesselRotator_01:Rotator)
    Scale="1,1,1"
   ) {
    rotationMode=BMM_Additive
    rotationSpace=BCS_BoneSpace
    translationSpace=BCS_BoneSpace
    alphaBoolBlend {blendOption=Linear}
    alphaScaleBiasClamp {interpSpeedDecreasing=2, interpSpeedIncreasing=2}
    boneToModify {boneName=dyn_scarf_l_base}
   }
   n57 = "Transform (Modify) Bone - Bone: dyn_scarf_r_base"(
    ComponentPose=@n56
    Alpha=Get Tassle 02 Alpha() -> (Tassle 02 Alpha:double)
    Translation="0,0,0"
    Rotation=Get TesselRotator_02() -> (TesselRotator_02:Rotator)
    Scale="1,1,1"
   ) {
    rotationMode=BMM_Additive
    rotationSpace=BCS_BoneSpace
    translationSpace=BCS_BoneSpace
    alphaBoolBlend {blendOption=Linear}
    alphaScaleBiasClamp {bInterpResult=true, interpSpeedDecreasing=2, interpSpeedIncreasing=2}
    boneToModify {boneName=dyn_scarf_r_base}
   }
   n58 = RigidBody(ComponentPose=@n57, OverrideWorldGravity=Get GravityOverride_tessel()) {
    bClampLinearTranslationLimitToRefPose=true
    bOverrideWorldGravity=true
    overridePhysicsAsset=&M_MED_LivelyDomino_Physics
    simulationSpace=BaseBoneSpace
    alphaBoolBlend {blendOption=Linear}
    baseBoneRef {boneName=neck_02}
    componentAppliedLinearAccClamp=<3000 3000 3000>
    componentLinearAccScale=<0.25 0.05 0.09>
    componentLinearVelScale=<0.25 0.05 0.09>
    overrideWorldGravity {z=-980}
    simSpaceSettings {
     maxAngularAcceleration=50
     maxAngularVelocity=1
     maxLinearAcceleration=300
     maxLinearVelocity=150
     velocityScaleZ=0.25
     worldAlpha=0.5
     externalLinearDragV=<0.2 0.2>
    }
   }
   n59 = Blend Poses by bool(
    BlendPose_0=Component To Local(ComponentPose=@n55)
    BlendPose_1=Component To Local(ComponentPose=@n58)
    bActiveValue=Get Is Skydive Bool Alpha() -> (Is Skydive Bool Alpha:bool)
   ) {blendType=Linear}
   n60 = Anim Dynamics - Bone: dyn_upperarm_sleeve_l(
    ComponentPose=Local To Component(LocalPose=Blend Poses by bool(BlendPose_0=@n52, BlendPose_1=@n59, bActiveValue=@n53) {blendType=Linear})
    Alpha=1
    ExternalForce="0,0,0"
   ) {
    angularDampingOverride=1
    angularSpringConstant=75
    bAngularSpring=true
    bGravityOverrideInSimSpace=true
    bOverrideAngularDamping=true
    simulationSpace=RootRelative
    alphaBoolBlend {blendOption=Linear}
    boundBone {boneName=dyn_upperarm_sleeve_l}
    componentAppliedLinearAccClamp=<1000 1000 1000>
    componentLinearAccScale=<0.3 0.3 0.3>
    componentLinearVelScale=<0.3 0.3 0.3>
    gravityOverride {z=-980}
    physicsBodyDefinitions [{
      sphereCollisionRadius=0
      boundBone {boneName=dyn_upperarm_sleeve_l}
      boxExtents=<10 20 20>
      constraintSetup {angularLimitsMax=<5 1 5>, angularLimitsMin=<-5 -5 -5>, angularTarget=<1 0 0>}
      localJointOffset=<1 0 0>
     }]
   }
   n61 = Anim Dynamics - Bone: dyn_upperarm_sleeve_r(ComponentPose=@n60, Alpha=1, ExternalForce="0,0,0") {
    angularDampingOverride=1
    angularSpringConstant=75
    bAngularSpring=true
    bGravityOverrideInSimSpace=true
    bOverrideAngularDamping=true
    simulationSpace=RootRelative
    alphaBoolBlend {blendOption=Linear}
    boundBone {boneName=dyn_upperarm_sleeve_r}
    componentAppliedLinearAccClamp=<1000 1000 1000>
    componentLinearAccScale=<0.3 0.3 0.3>
    componentLinearVelScale=<0.3 0.3 0.3>
    gravityOverride {z=-980}
    physicsBodyDefinitions [{
      sphereCollisionRadius=0
      boundBone {boneName=dyn_upperarm_sleeve_r}
      boxExtents=<10 20 20>
      constraintSetup {angularLimitsMax=<5 1 5>, angularLimitsMin=<-5 -5 -5>, angularTarget=<1 0 0>}
      localJointOffset=<-1 0 0>
     }]
   }
   n62 = Component To Local(ComponentPose=@n61)
   n63 = Get bRigidBodyEnabled()
   n64 = "Constraint - dyn_scarf_l_01 (PP)"(ComponentPose=Local To Component(LocalPose=Use cached pose 'control rig pose'()), Alpha=0.5, ConstraintWeights_0=1, ConstraintWeights_1=0) {
    alphaBoolBlend {blendOption=Linear}
    boneToModify {boneName=dyn_scarf_l_01}
    constraintSetup [{targetBone {boneName=spine_05}}, {targetBone {boneName=dyn_scarf_l_01}}]
    constraintWeights [1,1]
   }
   n65 = "Constraint - dyn_scarf_r_01 (PP)"(ComponentPose=@n64, Alpha=0.5, ConstraintWeights_0=1, ConstraintWeights_1=0) {
    alphaBoolBlend {blendOption=Linear}
    boneToModify {boneName=dyn_scarf_r_01}
    constraintSetup [{targetBone {boneName=spine_05}}, {targetBone {boneName=dyn_scarf_r_01}}]
    constraintWeights [1,1]
   }
   n66 = "Transform (Modify) Bone - Bone: dyn_scarf_l_01"(ComponentPose=@n65, Alpha=1, Translation="0.000,0.029,0.380", Rotation="0,0,0", Scale="1,1,1") {
    rotationMode=BMM_Additive
    rotationSpace=BCS_BoneSpace
    translationMode=BMM_Additive
    translationSpace=BCS_BoneSpace
    alphaBoolBlend {blendOption=Linear}
    alphaScaleBiasClamp {interpSpeedDecreasing=2, interpSpeedIncreasing=2}
    boneToModify {boneName=dyn_scarf_l_01}
    translation {y=0.029, z=0.38}
   }
   n67 = "Transform (Modify) Bone - Bone: dyn_scarf_r_01"(ComponentPose=@n66, Alpha=1, Translation="0,0,0", Rotation="0,0,0", Scale="1,1,1") {
    rotationMode=BMM_Additive
    rotationSpace=BCS_BoneSpace
    translationMode=BMM_Additive
    translationSpace=BCS_BoneSpace
    alphaBoolBlend {blendOption=Linear}
    alphaScaleBiasClamp {interpSpeedDecreasing=2, interpSpeedIncreasing=2}
    boneToModify {boneName=dyn_scarf_r_01}
   }
   n68 = "Constraint - dyn_scarf_l_02 (PP)"(ComponentPose=@n67, Alpha=0.5, ConstraintWeights_0=1, ConstraintWeights_1=0) {
    alphaBoolBlend {blendOption=Linear}
    boneToModify {boneName=dyn_scarf_l_02}
    constraintSetup [{targetBone {boneName=spine_05}}, {targetBone {boneName=dyn_scarf_l_01}}]
    constraintWeights [1,1]
   }
   n69 = "Constraint - dyn_scarf_r_02 (PP)"(ComponentPose=@n68, Alpha=0.5, ConstraintWeights_0=1, ConstraintWeights_1=0) {
    alphaBoolBlend {blendOption=Linear}
    boneToModify {boneName=dyn_scarf_r_02}
    constraintSetup [{targetBone {boneName=spine_05}}, {targetBone {boneName=dyn_scarf_r_01}}]
    constraintWeights [1,1]
   }
   n70 = "Transform (Modify) Bone - Bone: dyn_scarf_l_02"(ComponentPose=@n69, Alpha=1, Translation="0,1.2,0", Rotation="10.001,0.037,-0.022", Scale="1,1,1") {
    rotationMode=BMM_Additive
    rotationSpace=BCS_BoneSpace
    translationMode=BMM_Additive
    translationSpace=BCS_BoneSpace
    alphaBoolBlend {blendOption=Linear}
    alphaScaleBiasClamp {interpSpeedDecreasing=2, interpSpeedIncreasing=2}
    boneToModify {boneName=dyn_scarf_l_02}
    rotation=<rot 10.001 0.037 -0.022>
    translation {y=1.2}
   }
   n71 = "Transform (Modify) Bone - Bone: dyn_scarf_r_02"(ComponentPose=@n70, Alpha=1, Translation="0,0,0", Rotation="10.000,0.032,0.007", Scale="1,1,1") {
    rotationMode=BMM_Additive
    rotationSpace=BCS_BoneSpace
    scaleSpace=BCS_BoneSpace
    translationMode=BMM_Additive
    translationSpace=BCS_BoneSpace
    alphaBoolBlend {blendOption=Linear}
    alphaScaleBiasClamp {interpSpeedDecreasing=2, interpSpeedIncreasing=2}
    boneToModify {boneName=dyn_scarf_r_02}
    rotation=<rot 10 0.032 0.007>
   }
   n72 = "Constraint - dyn_scarf_l_03 (PP)"(ComponentPose=@n71, Alpha=0.5, ConstraintWeights_0=1, ConstraintWeights_1=0) {
    alphaBoolBlend {blendOption=Linear}
    boneToModify {boneName=dyn_scarf_l_03}
    constraintSetup [{targetBone {boneName=spine_04}}, {targetBone {boneName=dyn_scarf_l_02}}]
    constraintWeights [1,1]
   }
   n73 = "Constraint - dyn_scarf_r_03 (PP)"(ComponentPose=@n72, Alpha=0.5, ConstraintWeights_0=1, ConstraintWeights_1=0) {
    alphaBoolBlend {blendOption=Linear}
    boneToModify {boneName=dyn_scarf_r_03}
    constraintSetup [{targetBone {boneName=spine_04}}, {targetBone {boneName=dyn_scarf_r_02}}]
    constraintWeights [1,1]
   }
   n74 = "Transform (Modify) Bone - Bone: dyn_scarf_l_03"(ComponentPose=@n73, Alpha=1, Translation="-0.002,1.171,2.007", Rotation="-20.000,0.005,-0.017", Scale="1,1,1") {
    rotationMode=BMM_Additive
    rotationSpace=BCS_BoneSpace
    translationMode=BMM_Additive
    translationSpace=BCS_BoneSpace
    alphaBoolBlend {blendOption=Linear}
    alphaScaleBiasClamp {interpSpeedDecreasing=2, interpSpeedIncreasing=2}
    boneToModify {boneName=dyn_scarf_l_03}
    rotation=<rot -20 0.005 -0.017>
    translation=<-0.002 1.171 2.007>
   }
   n75 = "Transform (Modify) Bone - Bone: dyn_scarf_r_03"(ComponentPose=@n74, Alpha=1, Translation="0.000,0.302,2.058", Rotation="-20,0.016,-0.007", Scale="1,1,1") {
    rotationMode=BMM_Additive
    rotationSpace=BCS_BoneSpace
    scaleSpace=BCS_BoneSpace
    translationMode=BMM_Additive
    translationSpace=BCS_BoneSpace
    alphaBoolBlend {blendOption=Linear}
    alphaScaleBiasClamp {interpSpeedDecreasing=2, interpSpeedIncreasing=2}
    boneToModify {boneName=dyn_scarf_r_03}
    rotation=<rot -20 0.016 -0.007>
    translation {y=0.302, z=2.058}
   }
   n76 = "Transform (Modify) Bone - Bone: dyn_scarf_l_03"(ComponentPose=@n75, Alpha=1, Translation="0,0,-1", Rotation="-20.001,0.110,0.058", Scale="1,1,1") {
    rotationMode=BMM_Additive
    rotationSpace=BCS_BoneSpace
    translationMode=BMM_Additive
    translationSpace=BCS_BoneSpace
    alphaBoolBlend {blendOption=Linear}
    alphaScaleBiasClamp {interpSpeedDecreasing=2, interpSpeedIncreasing=2}
    boneToModify {boneName=dyn_scarf_l_03}
    rotation=<rot -20.001 0.11 0.058>
    translation {z=-1}
   }
   n77 = "Transform (Modify) Bone - Bone: dyn_scarf_r_03"(ComponentPose=@n76, Alpha=1, Translation="0,0,0", Rotation="-20,0,0", Scale="1,1,1") {
    rotationMode=BMM_Additive
    rotationSpace=BCS_BoneSpace
    scaleSpace=BCS_BoneSpace
    translationMode=BMM_Additive
    translationSpace=BCS_BoneSpace
    alphaBoolBlend {blendOption=Linear}
    alphaScaleBiasClamp {interpSpeedDecreasing=2, interpSpeedIncreasing=2}
    boneToModify {boneName=dyn_scarf_r_03}
    rotation {pitch=-20}
   }
   n78 = "Constraint - dyn_scarf_l_04 (PP)"!(ComponentPose=@n77, Alpha=0.5, ConstraintWeights_0=1, ConstraintWeights_1:float=0) -> (Pose) {
    alphaBoolBlend {blendOption=Linear}
    boneToModify {boneName=dyn_scarf_l_04}
    constraintSetup [{targetBone {boneName=spine_02}}, {targetBone {boneName=pelvis}}]
    constraintWeights [1,1]
   }
   n79 = "Constraint - dyn_scarf_r_04 (PP)"!(ComponentPose=@n78, Alpha=0.5, ConstraintWeights_0=1, ConstraintWeights_1:float=0) -> (Pose) {
    alphaBoolBlend {blendOption=Linear}
    boneToModify {boneName=dyn_scarf_r_04}
    constraintSetup [{targetBone {boneName=spine_02}}, {targetBone {boneName=pelvis}}]
    constraintWeights [1,1]
   }
   n80 = "Transform (Modify) Bone - Bone: dyn_scarf_l_04"(ComponentPose=@n79, Alpha=1, Translation="0.019,0.439,-0.433", Rotation="18.717,-1.965,-3.416", Scale="1,1,1") {
    rotationMode=BMM_Replace
    rotationSpace=BCS_ParentBoneSpace
    translationMode=BMM_Additive
    translationSpace=BCS_BoneSpace
    alphaBoolBlend {blendOption=Linear}
    alphaScaleBiasClamp {interpSpeedDecreasing=2, interpSpeedIncreasing=2}
    boneToModify {boneName=dyn_scarf_l_04}
    rotation=<rot 18.717 -1.965 -3.416>
    translation=<0.019 0.439 -0.433>
   }
   n81 = "Transform (Modify) Bone - Bone: dyn_scarf_r_04"(ComponentPose=@n80, Alpha=1, Translation="0,0,0", Rotation="19.299,1.324,-5.576", Scale="1,1,1") {
    rotationMode=BMM_Replace
    rotationSpace=BCS_ParentBoneSpace
    translationMode=BMM_Additive
    translationSpace=BCS_BoneSpace
    alphaBoolBlend {blendOption=Linear}
    alphaScaleBiasClamp {interpSpeedDecreasing=2, interpSpeedIncreasing=2}
    boneToModify {boneName=dyn_scarf_r_04}
    rotation=<rot 19.299 1.324 -5.576>
   }
   n82 = "Transform (Modify) Bone - Bone: dyn_scarf_l_04"(ComponentPose=@n81, Alpha=1, Translation="0,0,-4", Rotation="0,0,0", Scale="1,1,1") {
    rotationMode=BMM_Replace
    rotationSpace=BCS_ParentBoneSpace
    translationMode=BMM_Additive
    translationSpace=BCS_BoneSpace
    alphaBoolBlend {blendOption=Linear}
    alphaScaleBiasClamp {interpSpeedDecreasing=2, interpSpeedIncreasing=2}
    boneToModify {boneName=dyn_scarf_l_04}
    translation {z=-4}
   }
   n83 = "Transform (Modify) Bone - Bone: dyn_scarf_r_04"(ComponentPose=@n82, Alpha=1, Translation="0,0,-2", Rotation="0,0,0", Scale="1,1,1") {
    rotationMode=BMM_Replace
    rotationSpace=BCS_ParentBoneSpace
    translationMode=BMM_Additive
    translationSpace=BCS_BoneSpace
    alphaBoolBlend {blendOption=Linear}
    alphaScaleBiasClamp {interpSpeedDecreasing=2, interpSpeedIncreasing=2}
    boneToModify {boneName=dyn_scarf_r_04}
    translation {z=-2}
   }
   n84 = "Constraint - dyn_scarf_l_05 (PP)"!(ComponentPose=@n83, Alpha=0.5, ConstraintWeights_0=1, ConstraintWeights_1:float=0) -> (Pose) {
    alphaBoolBlend {blendOption=Linear}
    boneToModify {boneName=dyn_scarf_l_05}
    constraintSetup [{targetBone {boneName=pelvis}}, {targetBone {boneName=dyn_scarf_l_04}}]
    constraintWeights [1,1]
   }
   n85 = "Constraint - dyn_scarf_r_05 (PP)"!(ComponentPose=@n84, Alpha=0.5, ConstraintWeights_0=1, ConstraintWeights_1:float=0) -> (Pose) {
    alphaBoolBlend {blendOption=Linear}
    boneToModify {boneName=dyn_scarf_r_05}
    constraintSetup [{targetBone {boneName=pelvis}}, {targetBone {boneName=dyn_scarf_r_04}}]
    constraintWeights [1,1]
   }
   n86 = "Transform (Modify) Bone - Bone: dyn_scarf_l_05"(ComponentPose=@n85, Alpha=1, Translation="0,0,0", Rotation="8.980,-4.289,-3.379", Scale="1,1,1") {
    rotationMode=BMM_Replace
    rotationSpace=BCS_ParentBoneSpace
    translationMode=BMM_Additive
    translationSpace=BCS_BoneSpace
    alphaBoolBlend {blendOption=Linear}
    alphaScaleBiasClamp {interpSpeedDecreasing=2, interpSpeedIncreasing=2}
    boneToModify {boneName=dyn_scarf_l_05}
    rotation=<rot 8.98 -4.289 -3.379>
   }
   n87 = "Transform (Modify) Bone - Bone: dyn_scarf_r_05"(ComponentPose=@n86, Alpha=1, Translation="0,0,0", Rotation="9.784,-0.360,-2.071", Scale="1,1,1") {
    rotationMode=BMM_Replace
    rotationSpace=BCS_ParentBoneSpace
    translationMode=BMM_Additive
    translationSpace=BCS_BoneSpace
    alphaBoolBlend {blendOption=Linear}
    alphaScaleBiasClamp {interpSpeedDecreasing=2, interpSpeedIncreasing=2}
    boneToModify {boneName=dyn_scarf_r_05}
    rotation=<rot 9.784 -0.36 -2.071>
   }
   n88 = "Transform (Modify) Bone - Bone: dyn_scarf_l_05"(ComponentPose=@n87, Alpha=1, Translation="0,0,-4", Rotation="0,0,0", Scale="1,1,1") {
    rotationMode=BMM_Replace
    rotationSpace=BCS_ParentBoneSpace
    translationMode=BMM_Additive
    translationSpace=BCS_BoneSpace
    alphaBoolBlend {blendOption=Linear}
    alphaScaleBiasClamp {interpSpeedDecreasing=2, interpSpeedIncreasing=2}
    boneToModify {boneName=dyn_scarf_l_05}
    translation {z=-4}
   }
   n89 = "Transform (Modify) Bone - Bone: dyn_scarf_r_05"(ComponentPose=@n88, Alpha=1, Translation="0,0,-2", Rotation="0,0,0", Scale="1,1,1") {
    rotationMode=BMM_Replace
    rotationSpace=BCS_ParentBoneSpace
    translationMode=BMM_Additive
    translationSpace=BCS_BoneSpace
    alphaBoolBlend {blendOption=Linear}
    alphaScaleBiasClamp {interpSpeedDecreasing=2, interpSpeedIncreasing=2}
    boneToModify {boneName=dyn_scarf_r_05}
    translation {z=-2}
   }
   n90 = "Transform (Modify) Bone - Bone: dyn_scarf_l_04"(ComponentPose=@n89, Alpha=1, Translation="0,0,0", Rotation="0,0,0", Scale="1,1,1") {
    rotationMode=BMM_Replace
    rotationSpace=BCS_ParentBoneSpace
    translationMode=BMM_Additive
    translationSpace=BCS_BoneSpace
    alphaBoolBlend {blendOption=Linear}
    alphaScaleBiasClamp {interpSpeedDecreasing=2, interpSpeedIncreasing=2}
    boneToModify {boneName=dyn_scarf_l_04}
   }
   n91 = "Transform (Modify) Bone - Bone: dyn_scarf_r_04"(ComponentPose=@n90, Alpha=1, Translation="0,0,0", Rotation="0,0,0", Scale="1,1,1") {
    rotationMode=BMM_Replace
    rotationSpace=BCS_ParentBoneSpace
    translationMode=BMM_Additive
    translationSpace=BCS_BoneSpace
    alphaBoolBlend {blendOption=Linear}
    alphaScaleBiasClamp {interpSpeedDecreasing=2, interpSpeedIncreasing=2}
    boneToModify {boneName=dyn_scarf_r_04}
   }
   n92 = Layered blend per bone(
    BasePose=Blend Poses by bool(
     BlendPose_0=Component To Local(ComponentPose=@n91)
     BlendPose_1=Blend Poses by bool(BlendPose_0=@n62, BlendPose_1=@n21, bActiveValue=@n63) {blendType=Linear}
     bActiveValue=Get Is Backbling Cape() -> (Is Backbling Cape:bool)
    ) {blendType=Linear}
    BlendPoses_0=Slot 'DefaultSlot'()
    BlendWeights_0=Get CustomIdleIsPlayingFloat() -> (CustomIdleIsPlayingFloat:float)
   ) {layerSetup [{branchFilters [{boneName=dyn_scarf_f_02}, {boneName=dyn_scarf_f_r01}, {boneName=dyn_scarf_f_l01}]}]}
   Output Pose(Result=@n92) {layerGroup=None}
   n93 = Control Rig(Source=Copy Pose From Mesh() {bUseAttachedParent=true}) {
    alphaBoolBlend {blendOption=Linear}
    controlRigAssetReference {blueprintRigClass=&/M_MED_LivelyDomino_CtrlRig.M_MED_LivelyDomino_CtrlRig_C}
   }
   n94 = "Constraint - gravity (TT)"!(ComponentPose=Local To Component(LocalPose=@n93), Alpha=1, ConstraintWeights_0=0.3, ConstraintWeights_1:float=0.7) -> (Pose) {
    alphaBoolBlend {blendOption=Linear}
    boneToModify {boneName=gravity}
    constraintSetup [{transformType=Translation, targetBone {boneName=root}}, {transformType=Translation, targetBone {boneName=head}}]
    constraintWeights [1,1]
   }
   n95 = Get lowerarm_rot_up_l() -> (lowerarm_rot_up_l:double)
   n96 = "Constraint - sleeve_fix_l (PPP)"!(ComponentPose=@n94, Alpha=@n95, ConstraintWeights_0=0.8, ConstraintWeights_1:float=0.2, ConstraintWeights_2:float=0) -> (Pose) {
    alphaBoolBlend {blendOption=Linear}
    boneToModify {boneName=sleeve_fix_l}
    constraintSetup [{targetBone {boneName=upperarm_l}}, {targetBone {boneName=lowerarm_l}}, {targetBone {boneName=upperarm_twist_02_l}}]
    constraintWeights [1,1,1]
   }
   n97 = Get lowerarm_rot_up_r() -> (lowerarm_rot_up_r:double)
   n98 = "Constraint - sleeve_fix_r (PPP)"!(ComponentPose=@n96, Alpha=@n97, ConstraintWeights_0=0.8, ConstraintWeights_1:float=0.2, ConstraintWeights_2:float=0) -> (Pose) {
    alphaBoolBlend {blendOption=Linear}
    boneToModify {boneName=sleeve_fix_r}
    constraintSetup [{targetBone {boneName=upperarm_r}}, {targetBone {boneName=lowerarm_r}}, {targetBone {boneName=upperarm_twist_02_r}}]
    constraintWeights [1,1,1]
   }
   n99 = "Transform (Modify) Bone - Bone: sleeve_fix_l"(ComponentPose=@n98, Alpha=@n95, Translation="-10,4,0", Rotation="0,0,0", Scale="1,1,1") {
    rotationMode=BMM_Additive
    rotationSpace=BCS_BoneSpace
    scaleMode=BMM_Additive
    scaleSpace=BCS_BoneSpace
    translationMode=BMM_Additive
    translationSpace=BCS_BoneSpace
    alphaBoolBlend {blendOption=Linear}
    boneToModify {boneName=sleeve_fix_l}
    translation=<-10 4>
   }
   n100 = "Transform (Modify) Bone - Bone: sleeve_fix_r"(ComponentPose=@n99, Alpha=@n97, Translation="13,-5,0", Rotation="0,0,0", Scale="1,1,1") {
    rotationMode=BMM_Additive
    rotationSpace=BCS_BoneSpace
    scaleMode=BMM_Additive
    scaleSpace=BCS_BoneSpace
    translationMode=BMM_Additive
    translationSpace=BCS_BoneSpace
    alphaBoolBlend {blendOption=Linear}
    boneToModify {boneName=sleeve_fix_r}
    translation=<13 -5>
   }
   n101 = "Constraint - dyn_upperarm_sleeve_l (PP)"!(ComponentPose=@n100, Alpha=1, ConstraintWeights_0=1, ConstraintWeights_1:float=0) -> (Pose) {
    alphaBoolBlend {blendOption=Linear}
    boneToModify {boneName=dyn_upperarm_sleeve_l}
    constraintSetup [{targetBone {boneName=upperarm_l}}, {targetBone {boneName=clavicle_l}}]
    constraintWeights [1,1]
   }
   n102 = "Constraint - dyn_upperarm_sleeve_r (PP)"!(ComponentPose=@n101, Alpha=1, ConstraintWeights_0=1, ConstraintWeights_1:float=0) -> (Pose) {
    alphaBoolBlend {blendOption=Linear}
    boneToModify {boneName=dyn_upperarm_sleeve_r}
    constraintSetup [{targetBone {boneName=upperarm_r}}, {targetBone {boneName=clavicle_l}}]
    constraintWeights [1,1]
   }
   n103 = "Transform (Modify) Bone - Bone: dyn_scarf_f_01"(ComponentPose=@n102, Alpha=1, Translation="-2,0,2", Rotation="5,0,0", Scale="1,1,1") {
    rotationMode=BMM_Additive
    rotationSpace=BCS_BoneSpace
    scaleMode=BMM_Additive
    scaleSpace=BCS_BoneSpace
    translationMode=BMM_Additive
    translationSpace=BCS_BoneSpace
    alphaBoolBlend {blendOption=Linear}
    boneToModify {boneName=dyn_scarf_f_01}
    rotation {pitch=5}
    translation {x=-2, z=2}
   }
   n104 = "Transform (Modify) Bone - Bone: dyn_scarf_l_01"(ComponentPose=@n103, Alpha=1, Translation="0,0,0", Rotation="-35,0,0", Scale="1,1,1") {
    rotationMode=BMM_Additive
    rotationSpace=BCS_BoneSpace
    translationSpace=BCS_BoneSpace
    alphaBoolBlend {blendOption=Linear}
    alphaScaleBiasClamp {interpSpeedDecreasing=2, interpSpeedIncreasing=2}
    boneToModify {boneName=dyn_scarf_l_01}
    rotation {pitch=-35}
   }
   n105 = "Transform (Modify) Bone - Bone: dyn_scarf_r_01"(ComponentPose=@n104, Alpha=1, Translation="0,0,0", Rotation="-35,0,0", Scale="1,1,1") {
    rotationMode=BMM_Additive
    rotationSpace=BCS_BoneSpace
    translationSpace=BCS_BoneSpace
    alphaBoolBlend {blendOption=Linear}
    alphaScaleBiasClamp {interpSpeedDecreasing=2, interpSpeedIncreasing=2}
    boneToModify {boneName=dyn_scarf_r_01}
    rotation {pitch=-35}
   }
   n106 = "Transform (Modify) Bone - Bone: dyn_scarf_fix"(ComponentPose=@n105, Alpha=1, Translation="1.5,0,2", Rotation="0,0,0", Scale="1,1,1") {
    rotationMode=BMM_Additive
    rotationSpace=BCS_BoneSpace
    translationMode=BMM_Additive
    translationSpace=BCS_BoneSpace
    alphaBoolBlend {blendOption=Linear}
    alphaScaleBiasClamp {interpSpeedDecreasing=2, interpSpeedIncreasing=2}
    boneToModify {boneName=dyn_scarf_fix}
    translation {x=1.5, z=2}
   }
   Save cached pose 'control rig pose'(Pose=Component To Local(ComponentPose=@n106)) {cacheName=control rig pose}
  ]}
 EventGraph {nodes [
   n1 = Event BlueprintUpdateAnimation()
   > n2 = Sequence!(execute) -> (then_0, then_1, then_2:Exec, then_3:Exec, then_4:Exec, then_5:Exec, then_6:Exec, then_7:Exec, then_8:Exec)
   n3 = Sequence!(execute=@n2.then_2) -> (then_0, then_1, then_2, then_3, then_4)
   n4 = OR Boolean!(A=Get bIsSkydiveFloating(), B=Get bIsSkydiveDiveMode(), C:bool=Get bIsParachuteOpened()) -> (ReturnValue)
   > Set Is Skydiving Alpha(Is Skydiving Alpha:double="To Float (Boolean)"(InBool=@n4)) -> (Output_Get:double)
   > Set Is Skydive Bool Alpha(Is Skydive Bool Alpha:bool=@n4) -> (Output_Get:bool)
   n5 = Gravity Override State Function DEPRECATED(
    execute=@n3.then_1
    CharacterPartAnimInstance=Self-Reference()
    GameTimeInSeconds=GetGameTimeInSeconds()
    GravityOverrideParams=Get Gravity Override Params()
   )
   > Set GravityOverride_tessel(GravityOverride_tessel:Vector=@n5) -> (Output_Get:Vector)
   n6 = float * float(A=GetGameTimeInSeconds(), B=1.125) {is=title K2Node_PromotableOperator}
   n7 = RandomRotatorFunction!(
    execute=@n3.then_2
    GameTimeInSeconds:double=@n6
    PerlinScale:double=1
    Frequency:Rotator="5,2.8,5"
    Amplitude:Rotator="10,5,3"
    Offset:Rotator="-10,-3,0"
    TimeOffset:double=0
   ) -> (then, RandomRotator:Rotator)
   > Set TesselRotator_01(TesselRotator_01:Rotator=@n7) -> (Output_Get:Rotator)
   n8 = RandomRotatorFunction(execute=@n3.then_3, GameTimeInSeconds=@n6, PerlinScale=1, Frequency="4.5,1.8,4.5", Amplitude="10,5,3", Offset="-10,3,0", TimeOffset=0)
   > Set TesselRotator_02(TesselRotator_02:Rotator=@n8) -> (Output_Get:Rotator)
   n9 = Branch(execute=@n3.then_4, Condition=Get bIsInFrontEnd())
   > Set Tassle 01 Alpha(Tassle 01 Alpha:double=0.25) -> (Output_Get:double)
   > Set Tassle 02 Alpha(Tassle 02 Alpha:double=0.425) -> (Output_Get:double)
   > Set CustomIdleIsPlayingFloat(CustomIdleIsPlayingFloat:float=GetCurveValue(self=GetOwnerFrontEndAnimInstance(), CurveName=Frontend_Idle_Bypass)) -> (Output_Get:float)
   Set Tassle 01 Alpha(execute=@n9.else, Tassle 01 Alpha=0.7)
   > Set Tassle 02 Alpha(Tassle 02 Alpha=0.5)
   n10 = CheckForBackbling()
   n11 = Branch(execute=@n2.then_3, Condition=@n10)
   > Set Is Wearing Backbling(Is Wearing Backbling=true)
   Set Is Wearing Backbling(execute=@n11.else, Is Wearing Backbling=false)
   n12 = Branch(execute=@n2.then_4, Condition=@n10.bIsBackblingACape)
   > Set Is Backbling Cape(Is Backbling Cape:bool=true) -> (Output_Get:bool)
   Set Is Backbling Cape(execute=@n12.else, Is Backbling Cape=false)
   Is Valid(exec=@n2.then_5, InputObject=Get PlayerPawn() -> (PlayerPawn:Fort Player Pawn Athena Object Reference))
   > n13 = Branch(Condition=Get bIsInFrontEnd())
   Set WPO Globals Function(
    execute=@n13.else
    Material=Get Mat Instance Dynamic()
    Overlay Material=Get Overlay Mat Instance Dynamic()
    Skeletal Mesh=Get SkelMesh()
    "Local Noise Transform  Location"=Get Local Noise Transform Location() -> (Local Noise Transform Location:Vector)
    Local Noise Offset=Get Local Noise Offset()
   )
   > Set WPO Regional Function!(
    execute
    Material:Material Instance Dynamic Object Reference=Get Mat Instance Dynamic()
    Overlay Material:Material Instance Dynamic Object Reference=Get Overlay Mat Instance Dynamic()
    Skeletal Mesh:Skeletal Mesh Component Object Reference=Get SkelMesh()
    Velocity Transform:Name=foot_r
    Velocity Interp Speed:double=Get Velocity Interp Speed()
    Collision Transform:Name=calf_r
    Collision Transform Axis:EAxis Enum
    Invert Collision Transform Axis:bool=false
    Delta Time X:double=@n1.DeltaTimeX
    Previous Velocity Transform Location:"Vector (by ref)"=Get R1 Velocity Transform Location()
    Previous Velocity Transform Velocity:"Vector (by ref)"=Get R1 Velocity Transform Velocity()
    Region:int=1
   ) -> (then)
   > Set WPO Regional Function(
    Material=Get Mat Instance Dynamic()
    Overlay Material=Get Overlay Mat Instance Dynamic()
    Skeletal Mesh=Get SkelMesh()
    Velocity Transform=foot_l
    Velocity Interp Speed=Get Velocity Interp Speed()
    Collision Transform=calf_l
    Invert Collision Transform Axis=true
    Delta Time X=@n1.DeltaTimeX
    Previous Velocity Transform Location=Get R2 Velocity Transform Location()
    Previous Velocity Transform Velocity=Get R2 Velocity Transform Velocity()
    Region=2
   )
   > Set WPO Regional Function(
    Material=Get Mat Instance Dynamic()
    Overlay Material=Get Overlay Mat Instance Dynamic()
    Skeletal Mesh=Get SkelMesh()
    Velocity Transform=pelvis
    Velocity Interp Speed=Get Velocity Interp Speed()
    Collision Transform=pelvis
    Invert Collision Transform Axis=true
    Delta Time X=@n1.DeltaTimeX
    Previous Velocity Transform Location=Get R3 Velocity Transform Location() -> (R3 Velocity Transform Location:Vector)
    Previous Velocity Transform Velocity=Get R3 Velocity Transform Velocity() -> (R3 Velocity Transform Velocity:Vector)
    Region=2
   )
   > n14 = Branch(Condition=Get bIsSkydiveFloating())
   > n15 = Set WPO Pho Wind Intensity(WPO Pho Wind Intensity=1)
   SetScalarParameterValue(
    execute=[@n16.then, @n15.then]
    self=[Get Mat Instance Dynamic(), Get Overlay Mat Instance Dynamic()]
    ParameterName=ClothSim_WindStrength
    Value=Get WPO Pho Wind Intensity()
   )
   n16 = Set WPO Pho Wind Intensity(execute=@n14.else, WPO Pho Wind Intensity=0)
   n17 = Get SkelMesh()
   Set lowerarm_rot_up_l(
    execute=@n2.then_6
    lowerarm_rot_up_l:double=GetAlphaFromReferenceTransformDelta(SkeletalMeshComponent=@n17, BoneName=lowerarm_l, BaseName=upperarm_l, InputRange=(X=20,Y=90), TwistAxis="0,0,1")
   ) -> (Output_Get:double)
   > Set lowerarm_rot_up_r(lowerarm_rot_up_r:double=GetAlphaFromReferenceTransformDelta(SkeletalMeshComponent=@n17, BoneName=lowerarm_r, BaseName=upperarm_r, InputRange=(X=20,Y=90), TwistAxis="0,0,1")) -> (Output_Get:double)
   n18 = Branch(execute=@n2.then_7, Condition=OR Boolean!(A=Get bIsPlayingMeleeAnim(), B=Get bIsPlayingEmote()) -> (ReturnValue))
   > Set Sleeve Animdynamics Alpha(Sleeve Animdynamics Alpha:double=0.5) -> (Output_Get:double)
   Set Sleeve Animdynamics Alpha(execute=@n18.else, Sleeve Animdynamics Alpha=1)
   Set Is In FrontEnd Alpha(execute=@n2.then_8, Is In FrontEnd Alpha:double="To Float (Boolean)"(InBool=Get bIsInFrontEnd())) -> (Output_Get:double)
   n19 = Event OnWeaponChanged()
   Get Dynamic Material Instance(execute=[@n19.then, @n26.then], SkeletalMeshComponent=GetPartSkeletalMeshComponent())
   > n20 = Sequence!(execute) -> (then_0, then_1, then_2)
   n21 = Set SkelMesh(execute=@n20.then_2, SkelMesh=GetPartSkeletalMeshComponent())
   > n22 = Get Dynamic Material Instance(SkeletalMeshComponent=@n21, ElementIndex=1)
   > Set Mat Instance Dynamic(Mat Instance Dynamic=@n22)
   > n23 = GetOrCreateDynamicOverlayMaterial(MeshComponent=@n21, MaterialName=None)
   > Set Overlay Mat Instance Dynamic(Overlay Mat Instance Dynamic=@n23)
   Event BlueprintInitializeAnimation()
   > Set SkelMesh(SkelMesh=GetOwnerSkeletalMeshComponent())
   n24 = TryGetPawnOwner()
   > Is Valid(InputObject=@n24)
   > n25 = Cast To FortPlayerPawnAthena(Object=@n24)
   > Set PlayerPawn(PlayerPawn:Fort Player Pawn Athena Object Reference=@n25) -> (Output_Get:Fort Player Pawn Athena Object Reference)
   n26 = Event OnCharacterCustomizationLoaded()
   Get bIsSkyDiving()
  ]}
]
variables [
 CustomIdleIsPlayingFloat {type=number}
 Gravity Override Params {
  type=GravityOverrideParamsStruct
  default {
   emoteOrMelee {gravityOverride=<0 -50 -15>, windAmplitude=<50 -25 100>, windFrequency=<1.5 10 4>, windOffset=<0 -100 -50>}
   front_End {gravityOverride=<0 -50 -15>, windAmplitude=<50 -25 100>, windFrequency=<1.5 10 4>, windOffset=<0 -100 -50>}
   onGround_Moving {jointName=gravity, gravityOverride=<0 -50 15>, windAmplitude=<50 -25 100>, windFrequency=<1.8 13 6>, windOffset=<0 -100 -50>}
   onGround_Standing {jointName=gravity, gravityOverride=<0 -50 -5>, windAmplitude=<50 -25 100>, windFrequency=<1.8 13 6>, windOffset=<0 -100 -50>}
   skydive_FloatingDown {gravityOverride=<0 -700 150>}
   skydive_GoingUp {gravityOverride=<0 -300 150>}
   tactical_Sprint {jointName=gravity, gravityOverride=<0 -450 15>}
  }
 }
 GravityOverride_tessel {type=Vector}
 Is Backbling Cape {type=boolean}
 Is In FrontEnd Alpha {type=number}
 Is Skydive Bool Alpha {type=boolean}
 Is Skydiving Alpha {type=number}
 Is Wearing Backbling {type=boolean}
 Local Noise Offset {type=Vector}
 Local Noise Transform Location {type=Vector}
 lowerarm_rot_up_l {type=number}
 lowerarm_rot_up_r {type=number}
 Mat Instance Dynamic {type=MaterialInstanceDynamic}
 Overlay Mat Instance Dynamic {type=MaterialInstanceDynamic}
 PlayerPawn {type=FortPlayerPawnAthena}
 R1 Velocity Transform Location {type=Vector}
 R1 Velocity Transform Velocity {type=Vector}
 R2 Velocity Transform Location {type=Vector}
 R2 Velocity Transform Velocity {type=Vector}
 R3 Velocity Transform Location {type=Vector}
 R3 Velocity Transform Velocity {type=Vector}
 SkelMesh {type=SkeletalMeshComponent}
 Sleeve Animdynamics Alpha {type=number}
 Tassle 01 Alpha {type=number}
 Tassle 02 Alpha {type=number}
 TesselRotator_01 {type=Rotator}
 TesselRotator_02 {type=Rotator}
 Velocity Interp Speed {type=number}
 WPO Pho Wind Intensity {type=number}
]
